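; Gameplay ability for a Battle Pass S28 quest that tracks how long the player stays in Creepin' Cardboard.
; On event activation it calls the parent and starts a looping one-second timer, storing its handle.
; Each tick broadcasts a custom verb message with the owning player pawn and a configurable verb name, presumably counted by the quest.
; When the ability ends, it clears and invalidates the timer.
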(event Ability|EventActivateAbilityFromEvent
  (params (EventData :type Gameplay%20Event%20Data%20Structure%20%28by%20ref%29))
  (|CallParentFunction EventData)
  ;[1 1 1] Start Checking If In CreepinCardBoard
  (bind _returnvalue (Utilities|Time|SetTimerbyEvent (delegate Custom|Check%20In%20CreepinCardboard) 1.0 true))
  (Variables|Default|SetTimer%20Handle _returnvalue))

(event Custom|Check%20In%20CreepinCardboard
  (CustomVerb|BroadcastCustomVerbMessage (CustomVerb|MakeCustomVerbMessage (Variables|GAVars|Data|GetOwningPlayerPawn) (Variables|Default|GetVerb%20Name))))

(event Ability|EventOnEndAbility
  (params (bWasCancelled :type Boolean))
  (bind _timer_handle (Variables|Default|GetTimer%20Handle))
  ;[1 1 1] On CreepinCardboard Exit Clear Timer and End Ability
  (if (Utilities|Time|IsValidTimerHandle _timer_handle)
    (Utilities|Time|ClearandInvalidateTimerbyHandle _timer_handle)
    (else
      _))
  (Ability|EndAbility))
